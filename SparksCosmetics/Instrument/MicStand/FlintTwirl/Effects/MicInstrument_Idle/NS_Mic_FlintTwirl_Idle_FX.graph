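# Idle glow Niagara system for the FlintTwirl mic stand cosmetic, drawing layered blue-cyan sprites.
# MicGlow bursts four persistent particles, using partition and spawn group to pick color, size and camera offset, with VisibilityTag set per partition for four sprite renderers.
# Waveforms pulse halo size and color, while time-of-day curves scale alpha, dimmed when User.bIsFrontEndPreview is set.
# Light attributes are also written, though only sprite renderers are present.
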NiagaraSystem /SparksCosmetics/Instrument/MicStand/FlintTwirl/Effects/MicInstrument_Idle/NS_Mic_FlintTwirl_Idle_FX
dependencies {
 usedDataInterfaces [&NiagaraDataInterfaceCurve, &NiagaraDataInterfaceColorCurve]
 usedDynamicInputs [&ColorFromCurve]
 usedModules [
  &SystemState
  &self:SystemSpawnScript.SystemScriptSource.SystemScriptGraph.NiagaraNodeAssignment_1.SetVariables_#1
  &self:SystemSpawnScript.SystemScriptSource.SystemScriptGraph.NiagaraNodeAssignment_2.SetVariables_#2
 ]
}
emitters {MicGlow {
  emitterUpdateScript [
   &EmitterState()
   &SpawnBurst_Instantaneous(Spawn Count=4, Spawn Group=1, Use Spawn Probability=false, UseLoopCountLimit=true) {module=SpawnBurst_Instantaneous001}
  ]
  particleSpawnScript [
   &PartitionParticles(Number Of Partitions=5)
   /SetVariables(Particles.VisibilityTag=$Particles.PartitionParticles.Partition) {isSetParameters=true}
   &InitializeParticle(
    Color=&MergeHelper_ColorSwitch(
     0=&MergeHelper_ColorSwitch(
      0=<rgba 0 0.1892 1 0.72>
      1=<rgba 0 1.5016 5 0.5>
      2=<rgba 0.0135 2.9653 4 0.5>
      3=<rgba 0 1.853 5 1>
      Index=$Particles.PartitionParticles.Partition
      Num=::NewEnumerator2
     )
     1=&MergeHelper_ColorSwitch(
      0=<rgba 0 0.1892 1 0.72>
      1=<rgba 0 1.5016 5 0.5>
      2=<rgba 0.0135 2.9653 4 0.2>
      3=<rgba 0 1.853 5 1>
      Index=$Particles.PartitionParticles.Partition
      Num=::NewEnumerator2
     )
     2=?
     3=?
     Index=&SpawnGroup()
     Num=::NewEnumerator0
    )
    Position Offset=<-4 0 0>
    Sprite Size Mode=::NewEnumerator3
    Uniform Sprite Size=&MergeHelper_FloatSwitch(
     0=&MergeHelper_FloatSwitch(0=20, 1=12, 2=27, 3=25, Index=$Particles.PartitionParticles.Partition, Num=::NewEnumerator2)
     1=&MergeHelper_FloatSwitch(0=20, 1=10.5, 2=19, 3=18, Index=$Particles.PartitionParticles.Partition, Num=::NewEnumerator2)
     2=?
     3=?
     Index=&SpawnGroup()
     Num=::NewEnumerator0
    )
    UsePositionOffset=true
    Write Material Random=true
   )
   &CameraOffset(Camera Offset Amount=&MergeHelper_FloatSwitch(
     0=&MergeHelper_FloatSwitch(0=0, 1=-5, 2=-10, 3=10, Index=$Particles.PartitionParticles.Partition, Num=::NewEnumerator2)
     1=&MergeHelper_FloatSwitch(0=0, 1=5, 2=-10, 3=10, Index=$Particles.PartitionParticles.Partition, Num=::NewEnumerator2)
     2=?
     3=?
     Index=&SpawnGroup()
     Num=::NewEnumerator0
    ))
   &FN_Sprite_Attributes(
    Camera Offset=15
    Color=&MakeLinearColorFromVectorAndFloat(
     "Vector (RGB)"=&MakeVectorFromLinearColorRGB(Color=$Particles.Color)
     "Float (Alpha)"=&MaskFloatBySpawnGroup(Masked Float=0.25, Passthrough Float=0.12)
    )
    Create Camera Offset Attribute=true
    Create Color Attribute=true
    Create SpriteSize Attribute=true
    SpriteSize=&Multiply_Vector2D_ByFloat(Float=2.25, Vector2D=$Particles.SpriteSize)
    Write to Module Namespace=true
   )
   /SetVariables(Particles.HaloDynamic:Vector4f=&MakeVector4(X=1, Y=0, Z=1, W=0)) {isSetParameters=true}
  ]
  particleUpdateScript [
   &ParticleState(Kill Particles When Lifetime Has Elapsed=false)
   &ScaleSpriteSize(Scale Sprite Size Mode=::NewEnumerator0, Uniform Scale Factor=$System.EngineOwnerScale)
   &SolveForcesAndVelocity(Clamp Velocity=false, Limit Acceleration=false)
   /SetVariables(Particles.RayDynamicParameter:Vector4f=&MakeVector4(X=$Emitter.Age, Y=0.1, Z=0, W=0)) {isSetParameters=true}
   &ScaleColor(
    Scale Alpha=&Multiply_Float(
     A=&MergeHelper_FloatSwitch(
      0=&FloatFromCurve(
       FloatCurve=DataInterface() {propertyValues {Curve {keys [
           {interpMode=RCIM_Cubic, leaveTangentWeight=0.1001, tangentMode=RCTM_User, time=0, value=0.9}
           {arriveTangentWeight=0.1021, interpMode=RCIM_Cubic, leaveTangentWeight=0.0679, tangentMode=RCTM_User, time=0.3003, value=1}
           {arriveTangentWeight=0.0666, interpMode=RCIM_Cubic, leaveTangentWeight=0.0667, tangentMode=RCTM_User, time=0.5, value=1}
           {arriveTangent=-0.2, interpMode=RCIM_Cubic, leaveTangent=-0.2, time=0.7, value=0.95}
           {arriveTangentWeight=0.1, interpMode=RCIM_Cubic, tangentMode=RCTM_User, time=1, value=0.9}
          ]}}}
       CurveIndex=Linked() {linkedVariable {name=NPC.FortniteNPC.RemappedTimeOfDay_0-1, type=NiagaraFloat}}
       Scale Curve=&MakeCustomFloatFromBool(Bool=$User.bIsFrontEndPreview, True Float=0.7, False Float=1)
      )
      1=&FloatFromCurve(
       FloatCurve=DataInterface() {propertyValues {Curve {
          defaultValue=1
          keys [
           {interpMode=RCIM_Cubic, tangentMode=RCTM_User, time=0, value=0.9}
           {arriveTangentWeight=0.1021, interpMode=RCIM_Cubic, leaveTangentWeight=0.0679, tangentMode=RCTM_User, time=0.3003, value=1}
           {arriveTangentWeight=0.1667, interpMode=RCIM_Cubic, leaveTangentWeight=0.1667, tangentMode=RCTM_User, time=0.5, value=1}
           {arriveTangent=-0.2, interpMode=RCIM_Cubic, leaveTangent=-0.2, time=0.7, value=0.95}
           {arriveTangentWeight=0.3333, interpMode=RCIM_Cubic, tangentMode=RCTM_User, time=1, value=0.9}
          ]
         }}}
       CurveIndex=Linked() {linkedVariable {name=NPC.FortniteNPC.RemappedTimeOfDay_0-1, type=NiagaraFloat}}
       Scale Curve=&MakeCustomFloatFromBool(Bool=$User.bIsFrontEndPreview, True Float=0.7, False Float=1)
      )
      2=&FloatFromCurve(
       FloatCurve=DataInterface() {propertyValues {Curve {
          defaultValue=1
          keys [
           {interpMode=RCIM_Cubic, tangentMode=RCTM_User, time=0, value=0.85}
           {
            arriveTangent=0.3983
            arriveTangentWeight=0.1001
            interpMode=RCIM_Cubic
            leaveTangent=0.3983
            leaveTangentWeight=0.0666
            tangentMode=RCTM_User
            time=0.2225
            value=0.9562
           }
           {arriveTangentWeight=0.1667, interpMode=RCIM_Cubic, leaveTangentWeight=0.1667, tangentMode=RCTM_User, time=0.5, value=1}
           {arriveTangent=-0.5195, interpMode=RCIM_Cubic, leaveTangent=-0.5195, tangentMode=RCTM_User, time=0.7977, value=0.92}
           {arriveTangentWeight=0.3333, interpMode=RCIM_Cubic, tangentMode=RCTM_User, time=1, value=0.85}
          ]
         }}}
       CurveIndex=Linked() {linkedVariable {name=NPC.FortniteNPC.RemappedTimeOfDay_0-1, type=NiagaraFloat}}
       Scale Curve=&MakeCustomFloatFromBool(Bool=$User.bIsFrontEndPreview, True Float=0.5, False Float=1)
      )
      3=?
      Index=$Particles.PartitionParticles.Partition
      Num=::NewEnumerator1
     )
     B=&MaskFloatBySpawnGroup(Masked Float=1, Passthrough Float=1)
    )
    ScaleA=true
    ScaleRGB=true
   )
   /SetVariables(
    Particles.FN_Sprite_Attributes.SpriteSize:Vector2f=&Multiply_Vector2D_ByFloat(
     Float=&MakeFloatFromVector(Vector=$Engine.Owner.Scale, Channel=::NewEnumerator0)
     Vector2D=&Multiply_Vector2D_ByFloat(
      Float=&Waveform("Waveform [1]"=::NewEnumerator0, "[1] Phase Scale"=2.5, "[1] Frequency"=1, "[1] Amplitude Min/Max"=<0.98 1>)
      Vector2D=$Particles.Initial.FN_Sprite_Attributes.SpriteSize
     )
    )
    Particles.HaloColor:LinearColor=&Multiply_LinearColorByFloat(
     Linear Color=$Particles.FN_Sprite_Attributes.Color
     Float=&Waveform("Waveform [1]"=::NewEnumerator12, "[1] Frequency"=5, "[1] Amplitude Min/Max"=<0.85 1>)
    )
   ) {isSetParameters=true}
   &FN_Light_Attributes(
    Color=<rgba 0 1.853 5 1>
    Create Color Attribute=true
    Create Exponent Attribute=true
    Create Position Attribute=true
    Create Radius Attribute=true
    Exponent=3
    Position=<0 0 15.5>
    Radius=65
   )
  ]
  rendererClasses [&NiagaraSpriteRendererProperties]
  renderers [
   {rendererClass=&NiagaraSpriteRendererProperties}
   {rendererClass=&NiagaraSpriteRendererProperties, rendererIndex=1}
   {rendererClass=&NiagaraSpriteRendererProperties, rendererIndex=2}
   {rendererClass=&NiagaraSpriteRendererProperties, rendererIndex=3}
  ]
 }}
summary {
 systemName=NS_Mic_FlintTwirl_Idle_FX
 emitters [{emitterName=MicGlow, rendererClasses [&NiagaraSpriteRendererProperties]}]
 userVariables [User.bIsFrontEndPreview {type=NiagaraBool}]
}
